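# Unlit translucent Niagara sprite material for the smoke of a stinky cloud contrail effect.
# Two time-panned samples of one distortion texture multiply together to warp the UVs of a grayscale smoke texture.
# The squared smoke value blends from a ColorAdd tint plus faint noise to particle color, fed through MF_FX_EmissiveAdjustments.
# Opacity combines particle alpha, the smoke value and a depth fade.
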Material /Game/Effects/Fort_Effects/Gameplay/Cerberus/M_Contrail_StinkyCloud_Smoke
graph {
 properties {bUsedWithNiagaraSprites=true, blendMode=BLEND_Translucent, pixelDepthOffsetMode=PDOM_Legacy, shadingModel=MSM_Unlit, usedShadingModels=Unlit}
 nodes [
  n1 = VectorParameter() {parameterName=ColorAdd, defaultValue=<rgba 0.0554 0.0833 0.0108 1>}
  n2 = TextureCoordinate()
  n3 = Constant()
  n4 = Multiply(A=Time()) {constB=0.3}
  n5 = TextureObjectParameter() {parameterName=DistortionTexture, texture=/Game/Effects/Fort_Effects/Textures/T_Contrail_StinkyCloud_ColorAdditive}
  n6 = TextureSample(UVs=Add(A=Multiply(A=@n2) {constB=0.3}, B=AppendVector(A=@n3, B=@n4)), Tex=@n5) {texture=&T_Contrail_StinkyCloud_ColorAdditive}
  n7 = Multiply(A=@n6.R) {constB=0.2}
  n8 = ParticleColor()
  n9 = Multiply(A=TextureCoordinate()) {constB=0.5}
  n10 = Time()
  n11 = AppendVector(A=Multiply(A=@n10) {constB=0.1}, B=Multiply(A=@n10) {constB=0.25})
  n12 = Multiply(
   A=@n6.R
   B=TextureSample(UVs=Add(A=@n9, B=@n11), Tex=@n5) {texture=&T_Contrail_StinkyCloud_ColorAdditive}.G
  )
  n13 = TextureSampleParameter2D(UVs=Add(A=TextureCoordinate(), B=Multiply(A=@n12) {constB=0.4})) {parameterName=Texture, samplerType=SAMPLERTYPE_LinearGrayscale, texture=/Game/Effects/Fort_Effects/Textures/T_Contrail_StinkyCloud_Smoke}
  MaterialFunctionCall(Emissive=LinearInterpolate(A=Add(A=@n1.RGB, B=@n7), B=@n8.RGB, Alpha=Multiply(A=@n13.R, B=@n13.R))) {materialFunction=&MF_FX_EmissiveAdjustments}
  Multiply(A=Multiply(A=@n8.A, B=@n13.R), B=DepthFade() {fadeDistanceDefault=20})
 ]
}
