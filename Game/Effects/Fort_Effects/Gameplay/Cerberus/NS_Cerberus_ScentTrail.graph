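# Niagara system for the Cerberus scent trail: a warm, amber smoky trail with bubbles, flies, glow, sprite decals and a flickering light.
# Rising smoke puffs, GPU bubbles pushed by a turbulence vector field and a lighter ribbon form the body; a darker ribbon is disabled.
# Curl-noise mesh flies feed Fly_Trail ribbons through a particle attribute reader, with spawn rate tied to fly count.
# Several colors and the glow opacity vary with time of day.
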NiagaraSystem /Game/Effects/Fort_Effects/Gameplay/Cerberus/NS_Cerberus_ScentTrail
dependencies {
 usedDataInterfaces [
  &NiagaraDataInterfaceCurve
  &NiagaraDataInterfaceSpriteRendererInfo
  /Script/Niagara.NiagaraDataInterfaceVectorField
  &NiagaraDataInterfaceVectorCurve
  &NiagaraDataInterfaceParticleRead
  &NiagaraDataInterfaceColorCurve
 ]
 usedDynamicInputs []
 usedModules [&SystemState]
}
emitters {
 Bubbles {
  simTarget=GPUComputeSim
  emitterUpdateScript [
   &EmitterState()
   &SpawnRate(SpawnRate=10, Use Spawn Probability=false)
  ]
  particleSpawnScript [
   &InitializeParticle(
    Color=<rgba 1 0.66 0.25 1>
    Lifetime Max=3
    Lifetime Min=1.5
    Lifetime Mode=::NewEnumerator1
    Sprite Rotation Mode=::NewEnumerator2
    Sprite Size Mode=::NewEnumerator1
    Uniform Sprite Size Max=3
    Uniform Sprite Size Min=2
    UsePositionOffset=false
    Write Material Random=true
   )
   &ShapeLocation(MaskSpawnGroup=false, Override Seed=false, Sphere Radius=20, UseOwnerScale=true)
   &AddVelocity(
    Velocity=$Particles.ShapeLocation.ShapeVector
    Velocity Speed Scale=&RandomRangeFloat(Minimum=0.5, Maximum=0.6)
   )
  ]
  particleUpdateScript [
   &ParticleState()
   /Niagara/VectorFields/SampleVectorField(
    FieldScale:Vector3f=<8 8 8>
    VectorField=DataInterface() {propertyValues {Field=/Game/Effects/VectorFields/Noisy_Turbulence_103, bTileX=true, bTileY=true, bTileZ=true}}
   )
   /Niagara/VectorFields/ApplyVectorField(Vector Field Intensity as Force=50, Vector Field Velocity Tightness=0.1)
   &AccelerationForce(Acceleration=<0 0 100>)
   &ScaleColor(ScaleA=true, ScaleRGB=true)
   &ScaleSpriteSize(Uniform Curve Sprite Scale=DataInterface() {propertyValues {Curve {keys [
        {arriveTangent=-0.5095, time=0.0013, value=1}
        {arriveTangent=-0.9047, interpMode=RCIM_Cubic, leaveTangent=-0.9047, time=0.8, value=0.8}
        <at 0.9961 0.1>
       ]}}})
   &Drag(Use Linear Drag=true, Use Rotational Drag=true)
   &SolveForcesAndVelocity(Clamp Velocity=false, Limit Acceleration=false)
  ]
  rendererClasses [&NiagaraSpriteRendererProperties]
  renderers [{rendererClass=&NiagaraSpriteRendererProperties}]
 }
 Decal {
  emitterUpdateScript [
   &EmitterState(Life Cycle Mode=::NewEnumerator1, Loop Duration=2, UseLoopDelay=false)
   &SpawnRate(SpawnRate=12, Use Spawn Probability=false)
  ]
  particleSpawnScript [
   &InitializeParticle(
    Color=<rgba 0.1 0.0753 0 1>
    Lifetime Min=1.5
    Lifetime Mode=::NewEnumerator1
    Sprite Rotation Angle=&RandomRangeFloat(Minimum=0, Maximum=1)
    Sprite Rotation Mode=::NewEnumerator1
    Sprite Size Max=<100 100>
    Sprite Size Min=<75 75>
    Sprite Size Mode=::NewEnumerator2
    UsePositionOffset=false
    Write Material Random=true
   )
   &ShapeLocation(
    MaskSpawnGroup=false
    Offset=<0 0 10>
    Offset Mode=::NewEnumerator0
    Override Seed=false
    Sphere Radius=&Multiply_Float(A=25, B=$User.FXScale)
    UseOwnerScale=true
   )
   &DynamicMaterialParameters(
    Index 0 Param 1=1
    Index 0 Param 2=1
    Index 0 Param 3=1
    Index 0 Param 4=1
    Index 1 Param 1=1
    Index 1 Param 2=20
    Module.Param0WriteEnabled=true
    Module.Param0WriteEnabled1=true
    Module.Param1WriteEnabled=true
    Module.Param1WriteEnabled1=true
    Module.Param2WriteEnabled=true
    Module.Param2WriteEnabled1=true
    Module.Param3WriteEnabled=true
    Module.Param3WriteEnabled1=true
    Write Parameter Index 1=true
   )
   &CameraOffset(Camera Offset Amount=20)
  ]
  particleUpdateScript [
   &ParticleState()
   &ScaleColor(
    Scale Alpha=&FloatFromCurve(
     FloatCurve=DataInterface() {propertyValues {Curve {keys [
         {arriveTangent=0.1165, interpMode=RCIM_Cubic, leaveTangent=0.1165, tangentMode=RCTM_User, time=0, value=0}
         {interpMode=RCIM_Cubic, tangentMode=RCTM_Break, time=0.5, value=1}
         <at 1 0>
        ]}}}
     Scale Curve=&RandomRangeFloat(Minimum=0.75, Maximum=1)
    )
    Scale RGB=&VectorFromFloat(Value=&RandomRangeFloat(Minimum=1, Maximum=7))
    ScaleA=true
    ScaleRGB=true
   )
   &ScaleSpriteSize(Scale Factor=&Vector2DFromFloat(Value=0.7))
  ]
  rendererClasses [&NiagaraSpriteRendererProperties]
  renderers [{rendererClass=&NiagaraSpriteRendererProperties}]
 }
 Flies {
  emitterUpdateScript [
   &EmitterState(
    Inactive Response=::NewEnumerator1
    Life Cycle Mode=::NewEnumerator1
    Loop Duration=&RandomRangeFloat(Minimum=0.75, Maximum=1.5)
    UseLoopDelay=false
   )
   &SpawnRate(Spawn Probability=0.6, SpawnRate=6, Use Spawn Probability=true)
  ]
  particleSpawnScript [
   &InitializeParticle(
    Color Maximum=<rgba 0.0689 0.5663 0.6615 1>
    Color Minimum=&RandomRangeLinearColor(Minimum=<rgba 0 0.0057 0.057 1>, Maximum=<rgba 0.005 0.02 0 1>)
    Color Mode=::NewEnumerator2
    Lifetime Max=1.5
    Lifetime Mode=::NewEnumerator1
    Mesh Scale Mode=::NewEnumerator1
    Mesh Uniform Scale Max=0.2
    Mesh Uniform Scale Min=0.1
    UsePositionOffset=false
    Write Material Random=true
   )
   &CylinderLocation(Cylinder Height=1, Cylinder Height Midpoint=1, Cylinder Radius=40, MaskSpawnGroup=false, Override Seed=false, UseOwnerScale=true) {enabled=false}
   &ShapeLocation(
    Cylinder Height Midpoint=0
    Cylinder Radius=25
    MaskSpawnGroup=false
    Override Seed=false
    Shape Primitive=::NewEnumerator1
    Surface Only=false
    UseOwnerScale=true
   )
   &InitialMeshOrientation(Use Rotation Vector=false)
  ]
  particleUpdateScript [
   &InheritSourceMovement(Applied Position Scale=<0 0 0.7>, ApplyForce=false, ApplyPosition=true, ApplyPositionFalloff=false) {enabled=false}
   /SetVariables(Particles.CustomizePosition:Vector3f=&Multiply_Vector(
     A=&RandomRangeVector(Minimum=<100 100 5>, Maximum=<200 200 12>)
     B=&VectorFromCurve(
      VectorCurve=DataInterface() {propertyValues {
        XCurve {keys [<at 0 1>,<at 0.12 0>,<at 0.24 -1>,<at 0.36 0>,<at 0.48 1>,<at 0.6 0>,<at 0.72 -1>,<at 0.84 0>,<at 0.96 1>,<at 1.08 0>,<at 1.2 -1>,<at 1.32 0>,<at 1.44 1>,<at 1.56 0>]}
        YCurve {keys [<at 0 0>,<at 0.12 1>,<at 0.24 0>,<at 0.36 -1>,<at 0.48 0>,<at 0.6 1>,<at 0.72 0>,<at 0.84 -1>,<at 0.96 0>,<at 1.08 1>,<at 1.2 0>,<at 1.32 -1>,<at 1.44 0>,<at 1.56 1>]}
        ZCurve {keys [<at 0 1.0067>]}
       }}
      Scale Curve=&FN_FrontEndVector3Modifier(Gameplay=<1 1 1>, Front End=<3 3 3>)
     )
    )) {enabled=false, isSetParameters=true}
   /SetVariables(Particles.Velocity=$Particles.CustomizePosition) {enabled=false, isSetParameters=true}
   &ParticleState()
   &DynamicMaterialParameters(
    Index 0 Param 1=1
    Index 0 Param 2=&RandomRangeFloat(Minimum=3.5, Maximum=5)
    Index 0 Param 3=0.5
    Module.Param0WriteEnabled=true
    Module.Param1WriteEnabled=true
    Module.Param2WriteEnabled=true
    Module.Param3WriteEnabled=true
   )
   &CurlNoiseForce(Noise Strength=&RandomRangeFloat(Minimum=450, Maximum=600), Pan Noise=false)
   &PointAttractionForce(
    Attraction Radius=1000
    Attraction Strength=&RandomRangeFloat(Minimum=0.2, Maximum=1)
    Kill Within Radius=false
    Module.Use Falloff=true
    UseAttractorOffset=false
   )
   &SolveForcesAndVelocity(Clamp Velocity=false, Limit Acceleration=false)
   &ScaleMeshSize(Scale Factor=&VectorFromFloat(Value=&FloatFromCurve(FloatCurve=DataInterface() {propertyValues {Curve {keys [<at -0.0046 1.0113>,<at 0.5445 0.8054>,<at 1 0>]}}}, Scale Curve=2)))
   &DynamicMaterialParameters(
    Index 0 Param 1=&UniformRangedFloat(Minimum=4, Maximum=5)
    Index 0 Param 2=&FloatFromCurve(FloatCurve=DataInterface() {propertyValues {Curve {keys [
         {interpMode=RCIM_Cubic, tangentMode=RCTM_User, time=0, value=0}
         {arriveTangent=0.002, interpMode=RCIM_Cubic, leaveTangent=0.002, tangentMode=RCTM_User, time=0.1827, value=1}
         {arriveTangent=-1.2195, interpMode=RCIM_Cubic, leaveTangent=-1.2195, tangentMode=RCTM_User, time=0.9833, value=0.6448}
         <at 1.0011 0.002>
        ]}}})
    Index 1 Param 1=&UniformRangedFloat(Minimum=0, Maximum=30)
    Index 1 Param 2=1.1418
    Index 1 Param 3=&UniformRangedFloat(Minimum=0, Maximum=1)
    Index 1 Param 4=&UniformRangedFloat(Minimum=0, Maximum=1)
    Module.Param0WriteEnabled=false
    Module.Param0WriteEnabled1=true
    Module.Param1WriteEnabled=true
    Module.Param1WriteEnabled1=false
    Module.Param2WriteEnabled=false
    Module.Param2WriteEnabled1=true
    Module.Param3WriteEnabled=false
    Module.Param3WriteEnabled1=true
    Write Parameter Index 1=true
   ) {module=DynamicMaterialParameters001}
  ]
  rendererClasses [&NiagaraMeshRendererProperties]
  renderers [{rendererClass=&NiagaraMeshRendererProperties, rendererIndex=0}]
 }
 Fly_Trail {
  emitterUpdateScript [
   &EmitterState(Enable Distance Culling=true, MaxDistance=2500, Module.bUseMaxDistance=true, Module.bUseMinDistance=false, Scalability Mode=::NewEnumerator1)
   /SetVariables(Emitter.EmitterReader=DataInterface() {propertyValues {EmitterBinding {bindingMode=Other, emitterName=Flies}}}) {isSetParameters=true}
   &SpawnRate(
    SpawnRate=/Game/Effects/Niagara/DynamicInputs/Attributes/FN_GetEmitterParticleCount(AttributeReader:NiagaraDataInterfaceParticleRead=$Emitter.EmitterReader, Multiplier:NiagaraFloat=20)
    Use Spawn Probability=true
   )
  ]
  particleSpawnScript [
   &self:Sample_Location(New Particle Attribute Reader:NiagaraDataInterfaceParticleRead=$Emitter.EmitterReader) {module=Sample_Location}
   &InitializeRibbon(
    Color=<rgba 1 1 1 0.4>
    Lifetime=0.5
    Module.Write Color=true
    Module.Write Lifetime=true
    Module.Write Mass=true
    Module.Write Material Random=true
    Module.Write Position=false
    Module.Write RibbonTwist=false
    Module.Write RibbonWidth=true
    Ribbon Width=&RandomRangeFloat(Minimum=2, Maximum=6)
   )
   &AddVelocity(Velocity=&FN_FrontEndVector3Modifier(Gameplay=<0 0 0>, Front End=<-125 75 75>))
   &AddVelocity(Velocity=&FN_FrontEndVector3Modifier(Gameplay=<0 -200 0>, Front End=<0 0 75>)) {enabled=false, module=AddVelocity001}
  ]
  particleUpdateScript [
   &ParticleState()
   &SolveForcesAndVelocity(Clamp Velocity=false, Limit Acceleration=false)
   &ScaleColor(
    Scale Alpha=/Niagara/DynamicInputs/Math/LerpMultipleFloats(
     Float Count:ENiagara_MultipleLerpCount=::NewEnumerator2
     1:NiagaraFloat=0
     2:NiagaraFloat=1
     3:NiagaraFloat=1
     4:NiagaraFloat=0
     5:NiagaraFloat=0
     Alpha:NiagaraFloat=$Particles.NormalizedAge
    )
    ScaleA=true
    ScaleRGB=false
   )
   &ScaleRibbonWidth(Ribbon Width Scale=&FloatFromCurve(
     FloatCurve=DataInterface() {propertyValues {Curve {keys [<at 0 1 RCIM_Cubic>,<at 1 0>]}}}
     Scale Curve=&VectorLength(Vector=$Particles.InheritedScale)
    ))
  ]
  rendererClasses [&NiagaraRibbonRendererProperties]
  renderers [{rendererClass=&NiagaraRibbonRendererProperties}]
 }
 Glow {
  emitterUpdateScript [
   &EmitterState()
   &SpawnRate(Use Spawn Probability=false)
  ]
  particleSpawnScript [
   &InitializeParticle(
    Color=<rgba 0.4 0.2 0 0.5>
    Lifetime=2
    Sprite Rotation Mode=::NewEnumerator2
    Sprite Size Mode=::NewEnumerator1
    Uniform Sprite Size Max=500
    Uniform Sprite Size Min=400
    UsePositionOffset=true
    Write Material Random=true
   )
   &ShapeLocation(MaskSpawnGroup=false, Offset Mode=::NewEnumerator0, Override Seed=false, Sphere Radius=0, UseOwnerScale=true)
  ]
  particleUpdateScript [
   &ParticleState()
   &ScaleSpriteSize(
    Uniform Curve Scale=&RemapRange(Input Value To Remap=1, Input Min=1, Input Max=5, Output Min=1, Output Max=3)
    Uniform Curve Sprite Scale=DataInterface() {propertyValues {Curve {keys [
        {tangentMode=RCTM_User, time=0, value=1}
        {interpMode=RCIM_Cubic, tangentMode=RCTM_User, time=0.2, value=1}
        {arriveTangent=-2.3246, interpMode=RCIM_Cubic, leaveTangent=-2.3246, tangentMode=RCTM_User, time=1, value=0.2}
       ]}}}
   )
   &SolveForcesAndVelocity(Clamp Velocity=false, Limit Acceleration=false)
   &CameraOffset(Camera Offset Amount=15)
   &ScaleColor(
    Scale Alpha=&FloatFromCurve(
     FloatCurve=DataInterface() {propertyValues {Curve {keys [
         {interpMode=RCIM_Cubic, tangentMode=RCTM_User, time=0, value=0}
         {interpMode=RCIM_Cubic, tangentMode=RCTM_User, time=0.5, value=1}
         {interpMode=RCIM_Cubic, tangentMode=RCTM_User, time=1, value=0}
        ]}}}
     Scale Curve=&FNTimeOfDayScalar(Night Scale=0.2, Morning Scale=0.3, Day Scale=0.6, Evening Scale=0.4)
    )
    ScaleA=true
    ScaleRGB=false
   )
   &DynamicMaterialParameters(
    Index 1 Param 1=150
    Module.Param0WriteEnabled1=true
    Module.Param1WriteEnabled1=true
    Module.Param2WriteEnabled1=true
    Module.Param3WriteEnabled1=true
    Write Parameter Index 0=false
    Write Parameter Index 1=true
   )
  ]
  rendererClasses [&NiagaraSpriteRendererProperties]
  renderers [{rendererClass=&NiagaraSpriteRendererProperties}]
 }
 Hand_FeetSmoke {
  emitterUpdateScript [
   &EmitterState()
   &SpawnRate(SpawnRate=5, Use Spawn Probability=false)
   &SpawnBurst_Instantaneous(Spawn Count=4, Spawn Probability=HlslExpression() {hlslExpression="Emitter.LoopCount == 0"}, Use Spawn Probability=true) {enabled=false}
  ]
  particleSpawnScript [
   &InitializeParticle(
    Color=&FNTimeOfDayColor(
     Night Color=<rgba 0.25 0.1624 0.0396 1>
     Morning Color=<rgba 0.25 0.2149 0.0396 1>
     Day Color=<rgba 0.5 0.3598 0.0793 1>
     Evening Color=<rgba 0.25 0.1624 0.0396 1>
    )
    Lifetime Max=1.5
    Lifetime Mode=::NewEnumerator1
    Sprite Rotation Mode=::NewEnumerator2
    Sprite Size Mode=::NewEnumerator3
    Uniform Sprite Size=&RandomRangeFloat(Minimum=50, Maximum=100)
    UsePositionOffset=false
    Write Material Random=true
   ) {module=InitializeParticle001}
   &CameraOffset(Camera Offset Amount=15)
  ]
  particleUpdateScript [
   &ParticleState()
   &ScaleSpriteSize(Uniform Curve Sprite Scale=DataInterface() {propertyValues {Curve {keys [
        {arriveTangent=2.8307, interpMode=RCIM_Cubic, leaveTangent=2.8307, tangentMode=RCTM_User, time=0, value=0}
        {interpMode=RCIM_Cubic, tangentMode=RCTM_User, time=1, value=1}
       ]}}})
   &ScaleColor(
    Scale Alpha=&FloatFromCurve(FloatCurve=DataInterface() {propertyValues {Curve {keys [
         {interpMode=RCIM_Cubic, leaveTangent=5.4482, tangentMode=RCTM_Break, time=0, value=0.2}
         {interpMode=RCIM_Cubic, leaveTangent=-1.8357, tangentMode=RCTM_Break, time=0.2, value=1}
         {arriveTangent=-1.4912, interpMode=RCIM_Cubic, tangentMode=RCTM_Break, time=1, value=0}
        ]}}})
    ScaleA=true
    ScaleRGB=true
   )
  ]
  rendererClasses [&NiagaraSpriteRendererProperties]
  renderers [{rendererClass=&NiagaraSpriteRendererProperties}]
 }
 Light {
  emitterUpdateScript [
   &EmitterState(
    Enable Distance Culling=true
    Life Cycle Mode=::NewEnumerator1
    Loop Duration=1
    MaxDistance:NiagaraFloat=&Multiply_Float(A=$User.FXDistance, B=0.7)
    Module.bUseMaxDistance=true
    Module.bUseMinDistance=true
    Scalability Mode=::NewEnumerator1
    UseLoopDelay=false
   )
   &SpawnBurst_Instantaneous(Use Spawn Probability=false, UseLoopCountLimit=false)
  ]
  particleSpawnScript [&InitializeParticle(
    Color=&FNTimeOfDayColor(Night Color=<rgba 3 2.5 0 1>, Morning Color=<rgba 1.8638 3 0 1>, Day Color=<rgba 1.8638 3 0 1>, Evening Color=<rgba 3 2.5 0 1>)
    Position Offset=<0 0 50>
    Sprite Size Mode=::NewEnumerator3
    Uniform Sprite Size=200
    UsePositionOffset=true
    Write Material Random=true
   )]
  particleUpdateScript [
   &ParticleState()
   &SolveForcesAndVelocity(Clamp Velocity=false, Limit Acceleration=false)
   &CameraOffset(Camera Offset Amount=10)
   &FN_Light_Attributes(Create Exponent Attribute=true, Create Radius Attribute=true, Exponent=2, Radius=100)
   &ScaleColor(
    Scale RGB=&VectorFromFloat(Value=&Waveform("Waveform [1]"=::NewEnumerator12, "[1] Frequency"=10, "[1] Amplitude Min/Max"=<0 1>, "[1] Scale"=0.2, "[1] Bias"=1))
    ScaleA=true
    ScaleRGB=true
   )
  ]
  rendererClasses [&NiagaraLightRendererProperties]
  renderers [{rendererClass=&NiagaraLightRendererProperties, rendererIndex=0}]
 }
 RibbonDark {
  bEnabled=false
  emitterUpdateScript [
   &EmitterState()
   &FN_SpawnRibbon_v2(Spawn Mode=::NewEnumerator0, SpawnRate=120)
  ]
  particleSpawnScript [
   &FN_InitializeRibbon_v3(Color:LinearColor=<rgba 0.9419 1 0.9242 0.8>, Lifetime=0.7, Ribbon Width=200) {module=FN_InitializeRibbon_v3001}
   &AddVelocity(Velocity=&MakeCustomVectorFromBool(Bool=$Particles.Locked, True Vector=<0 0 0>, False Vector=<0 0 450>)) {module=AddVelocity001}
  ]
  particleUpdateScript [
   &SolveForcesAndVelocity(Clamp Velocity=false, Limit Acceleration=false)
   &FN_UpdateRibbon(Initial Color:LinearColor=$Particles.Initial.Color, Initial RibbonWidth:NiagaraFloat=$Particles.Initial.RibbonWidth)
   &ScaleColor(
    Linear Color Curve=DataInterface() {propertyValues {
      AlphaCurve {keys [<at 0 0>,<at 0.0367 0.8>,<at 0.3459 1>,<at 0.8896 0.1>]}
      BlueCurve {keys [<at 0.0013 0>, <at 0.3474 1>, {interpMode=RCIM_Cubic, leaveTangent=-0.7216, time=0.9661, value=0}]}
      GreenCurve {keys [<at 0.0013 0>, <at 0.3474 1>, {interpMode=RCIM_Cubic, leaveTangent=-0.6443, time=0.9661, value=0}]}
      RedCurve {keys [
        <at 0.0013 0>
        <at 0.3474 1>
        {arriveTangent=-0.3741, interpMode=RCIM_Cubic, leaveTangent=-0.3741, tangentMode=RCTM_User, time=0.9661, value=0}
       ]}
     }}
    Scale Mode=::NewEnumerator2
   ) {module=ScaleColor001}
   &ScaleRibbonWidth(Ribbon Width Scale=&FloatFromCurve(FloatCurve=DataInterface() {propertyValues {Curve {keys [{interpMode=RCIM_Cubic, tangentMode=RCTM_User, time=0, value=0.25}, {interpMode=RCIM_Cubic, tangentMode=RCTM_User, time=0.5, value=1}]}}}))
  ]
  rendererClasses [&NiagaraRibbonRendererProperties]
  renderers [{rendererClass=&NiagaraRibbonRendererProperties}]
 }
 RibbonLighter {
  emitterUpdateScript [
   &EmitterState()
   &FN_SpawnRibbon_v2(Spawn Mode=::NewEnumerator0, SpawnRate=120)
  ]
  particleSpawnScript [
   &FN_InitializeRibbon_v3(
    Color:LinearColor=&FNTimeOfDayColor(
     Night Color=<rgba 0.25 0.1448 0.0396 1>
     Morning Color=<rgba 0.25 0.2149 0.0396 1>
     Day Color=<rgba 0.25 0.2149 0.0396 1>
     Evening Color=<rgba 0.25 0.2149 0.0396 1>
    )
    Ribbon Width=100
   ) {module=FN_InitializeRibbon_v3001}
   &AddVelocity(Velocity=&MakeCustomVectorFromBool(Bool=$Particles.Locked, True Vector=<0 0 0>, False Vector=<0 0 450>)) {module=AddVelocity001}
  ]
  particleUpdateScript [
   &SolveForcesAndVelocity(Clamp Velocity=false, Limit Acceleration=false)
   &FN_UpdateRibbon(Initial Color:LinearColor=$Particles.Initial.Color, Initial RibbonWidth:NiagaraFloat=$Particles.Initial.RibbonWidth)
   &ScaleColor(
    Linear Color Curve=DataInterface() {propertyValues {
      AlphaCurve {keys [<at 0 0>,<at 0.0508 0.6>,<at 0.7316 0.5>,<at 1 0.1>]}
      BlueCurve {keys [<at 0.0386 1>]}
      GreenCurve {keys [<at 0.0386 1>]}
      RedCurve {keys [<at 0.0386 1>]}
     }}
    Scale Mode=::NewEnumerator2
   ) {module=ScaleColor001}
   &ScaleRibbonWidth(Ribbon Width Scale=&FloatFromCurve(FloatCurve=DataInterface() {propertyValues {Curve {keys [{interpMode=RCIM_Cubic, tangentMode=RCTM_User, time=0, value=0.25}, {interpMode=RCIM_Cubic, tangentMode=RCTM_User, time=0.5, value=1}]}}}))
  ]
  rendererClasses [&NiagaraRibbonRendererProperties]
  renderers [{rendererClass=&NiagaraRibbonRendererProperties}]
 }
 Smoke_Fill {
  emitterUpdateScript [
   &EmitterState()
   &SpawnRate(SpawnRate=40, Use Spawn Probability=false)
   &SpawnBurst_Instantaneous(Spawn Count=7, Use Spawn Probability=false, UseLoopCountLimit=false) {enabled=false}
  ]
  particleSpawnScript [
   &InitializeParticle(
    Color=&FNTimeOfDayColor(
     Night Color=<rgba 0.4 0.2576 0.0582 0.4>
     Morning Color=<rgba 0.4 0.2861 0.0582 0.4>
     Day Color=<rgba 0.4 0.2861 0.0582 0.4>
     Evening Color=<rgba 0.4 0.2576 0.0582 0.4>
    )
    Lifetime Max=3.5
    Lifetime Min=3
    Lifetime Mode=::NewEnumerator1
    Sprite Rotation Mode=::NewEnumerator2
    Sprite Size Mode=::NewEnumerator1
    Uniform Sprite Size Max=100
    Uniform Sprite Size Min=75
    UsePositionOffset=false
    Write Material Random=true
   )
   &ShapeLocation(MaskSpawnGroup=false, Override Seed=false, Sphere Radius=10, Sphere Surface Distribution=0.25, UseOwnerScale=true)
   &AddVelocity(
    Velocity=&Multiply_VectorByFloat(Vector=$Particles.ShapeLocation.ShapeVector, Float=&RandomRangeFloat(Minimum=0.5, Maximum=0.75))
    Velocity Speed Scale=&RandomRangeFloat(Minimum=0.5, Maximum=0.75)
   )
  ]
  particleUpdateScript [
   &ParticleState()
   &DynamicMaterialParameters(
    Index 0 Param 1=&FloatFromCurve(
     FloatCurve=DataInterface() {propertyValues {Curve {keys [{arriveTangent=2.7533, time=0, value=1}, {arriveTangent=0.0529, leaveTangent=0.2021, time=1, value=1.5}]}}}
     CurveIndex=&Add_Float(A=&Multiply_Float(A=$Particles.Initial.MaterialRandom, B=0.5), B=$Particles.NormalizedAge)
    )
    Index 0 Param 2=&FloatFromCurve(FloatCurve=DataInterface() {propertyValues {Curve {keys [
         {arriveTangent=0.0453, interpMode=RCIM_Cubic, tangentMode=RCTM_Break, time=0, value=-1}
         {arriveTangent=0.0604, interpMode=RCIM_Cubic, leaveTangent=0.0604, tangentMode=RCTM_User, time=1, value=-1}
        ]}}})
    Index 0 Param 3=&FloatFromCurve(
     FloatCurve=DataInterface() {propertyValues {Curve {keys [{arriveTangent=-4.113, interpMode=RCIM_Cubic, time=0, value=1}, {interpMode=RCIM_Cubic, leaveTangent=0.7585, time=0.25, value=0}]}}}
     Scale Curve=0.25
    )
    Index 0 Param 4=&FloatFromCurve(FloatCurve=DataInterface() {propertyValues {Curve {keys [<at 0 1>,<at 1 0>]}}})
    Index 1 Param 1=&FloatFromCurve(
     FloatCurve=DataInterface() {propertyValues {Curve {keys [
         {interpMode=RCIM_Cubic, leaveTangent=4.7088, tangentMode=RCTM_Break, time=0, value=0}
         {arriveTangent=4.2257, interpMode=RCIM_Cubic, leaveTangent=4.2257, tangentMode=RCTM_User, time=0.25, value=1}
        ]}}}
     Scale Curve=&RandomRangeFloat(Minimum=1.5, Maximum=1.5)
    )
    Index 1 Param 2=&FloatFromCurve(
     FloatCurve=DataInterface() {propertyValues {Curve {keys [<at 0 0>, {leaveTangent=0.9066, time=1, value=1}]}}}
     Scale Curve=&RandomRangeFloat(Minimum=0.5, Maximum=0.6)
    )
    Index 1 Param 3=0.25
    Index 1 Param 4=&FloatFromCurve(
     FloatCurve=DataInterface() {propertyValues {Curve {keys [{arriveTangent=-0.3224, interpMode=RCIM_Cubic, leaveTangent=-0.3224, tangentMode=RCTM_User, time=0, value=1}, <at 1 0.9>]}}}
     Scale Curve=&RandomRangeFloat(Minimum=1.1, Maximum=1.4)
    )
    Index 2 Param 1=&FloatFromCurve(FloatCurve=DataInterface() {propertyValues {Curve {keys [{interpMode=RCIM_Cubic, leaveTangent=1.4673, tangentMode=RCTM_Break, time=0, value=0}, <at 1 1>]}}})
    Index 2 Param 2=&FloatFromCurve(
     FloatCurve=DataInterface() {propertyValues {Curve {keys [
         {arriveTangent=1.9056, interpMode=RCIM_Cubic, leaveTangent=1.9056, tangentMode=RCTM_User, time=0, value=0.2}
         {arriveTangent=0.0259, interpMode=RCIM_Cubic, leaveTangent=0.0259, tangentMode=RCTM_User, time=1, value=1.4}
        ]}}}
     Scale Curve=&RandomRangeFloat(Minimum=0.1, Maximum=0.4)
    )
    Index 2 Param 3=&FloatFromCurve(
     FloatCurve=DataInterface() {propertyValues {Curve {keys [
         {interpMode=RCIM_Cubic, leaveTangent=2.6052, tangentMode=RCTM_Break, time=0, value=1}
         {interpMode=RCIM_Cubic, leaveTangent=11.1233, tangentMode=RCTM_Break, time=1, value=2}
        ]}}}
     Scale Curve=2
    )
    Index 2 Param 4=&FloatFromCurve(FloatCurve=DataInterface() {propertyValues {Curve {keys [{interpMode=RCIM_Cubic, tangentMode=RCTM_Break, time=0, value=1}, {interpMode=RCIM_Cubic, tangentMode=RCTM_Break, time=1, value=1.1}]}}})
    Module.Param0WriteEnabled=true
    Module.Param0WriteEnabled1=false
    Module.Param0WriteEnabled2=true
    Module.Param1WriteEnabled=false
    Module.Param1WriteEnabled1=false
    Module.Param1WriteEnabled2=true
    Module.Param2WriteEnabled=true
    Module.Param2WriteEnabled1=true
    Module.Param2WriteEnabled2=true
    Module.Param3WriteEnabled=true
    Module.Param3WriteEnabled1=true
    Module.Param3WriteEnabled2=true
    Write Parameter Index 1=true
    Write Parameter Index 2=true
   )
   &ScaleColor(
    Scale Alpha=&FloatFromCurve(FloatCurve=DataInterface() {propertyValues {Curve {keys [<at 0 0 RCIM_Cubic>,<at 0.25 1 RCIM_Cubic>,<at 1 0>]}}})
    Scale RGB=&Multiply_VectorByFloat(
     Vector=<1 1 1>
     Float=&FloatFromCurve(FloatCurve=DataInterface() {propertyValues {Curve {keys [<at 0.25 1>]}}})
    )
    ScaleA=true
    ScaleRGB=true
   )
   &ScaleSpriteSize(Uniform Curve Sprite Scale=DataInterface() {propertyValues {Curve {keys [{arriveTangent=2.3114, interpMode=RCIM_Cubic, time=0, value=0.3}, {interpMode=RCIM_Cubic, leaveTangent=0.2, time=1, value=1}]}}})
   &AccelerationForce(Acceleration=<0 0 200>)
   &Drag(Drag=&RandomRangeFloat(Minimum=1.5, Maximum=2.5), Use Linear Drag=true, Use Rotational Drag=true)
   &SolveForcesAndVelocity(Clamp Velocity=false, Limit Acceleration=false)
   &SpriteRotationRate(Rotation Rate=&RandomRangeFloat(Minimum=-90, Maximum=90))
   &SubUVAnimation(
    End Frame Range Override=0
    Play Rate=5
    Random Start Frame=true
    "Sprite Renderer "=DataInterface() {propertyValues {SpriteRenderer=&self:Smoke_Fill_0.NiagaraSpriteRendererProperties_0}}
    SubUV Animation Mode:ENiagara_SubUVLookupModeV2=::NewEnumerator3
    UseEndFrame=false
    UsePlayRateScale=false
    UseStartFrame=false
   )
  ]
  rendererClasses [&NiagaraSpriteRendererProperties]
  renderers [{rendererClass=&NiagaraSpriteRendererProperties}]
 }
}
summary {
 systemName=NS_Cerberus_ScentTrail
 emitters [
  {emitterName=Smoke_Fill, rendererClasses [&NiagaraSpriteRendererProperties]}
  {emitterName=Bubbles, simTarget=GPUComputeSim, rendererClasses [&NiagaraSpriteRendererProperties]}
  {emitterName=Flies, rendererClasses [&NiagaraMeshRendererProperties]}
  {emitterName=Fly_Trail, rendererClasses [&NiagaraRibbonRendererProperties]}
  {emitterName=RibbonLighter, rendererClasses [&NiagaraRibbonRendererProperties]}
  {bEnabled=false, emitterName=RibbonDark, rendererClasses [&NiagaraRibbonRendererProperties]}
  {emitterName=Hand_FeetSmoke, rendererClasses [&NiagaraSpriteRendererProperties]}
  {emitterName=Decal, rendererClasses [&NiagaraSpriteRendererProperties]}
  {emitterName=Light, rendererClasses [&NiagaraLightRendererProperties]}
  {emitterName=Glow, rendererClasses [&NiagaraSpriteRendererProperties]}
 ]
 userVariables [
  User.bIsFrontEnd {type=NiagaraBool}
  User.bIsFrontEndPreview {type=NiagaraBool}
  User.FXColorRed {type=NiagaraBool}
  User.FXDistance {type=NiagaraFloat, defaultValue=8000, description=""}
  User.FXScale {type=NiagaraFloat, defaultValue=1, description=""}
  User.RibbonIDOffset {type=NiagaraInt32}
  User.SkeletalMesh {
   type=NiagaraDataInterfaceSkeletalMesh
   defaultValue {
    struct=NiagaraExt_VariableValue_DataInterface
    value {dataInterface=&self:NiagaraDataInterfaceSkeletalMesh_0}
   }
  }
 ]
}
